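# Sound cue for the Masamune stand opening, layering bell, blade scrape, foley and chest latch sounds with item-interact attenuation.
# A quality node sends its first two levels to a mixer with randomized bell, scrape and foley variants, and its last two to a leaner fixed-sample mix.
# The randomized bell's volume envelope holds it near silent briefly before ramping up.
# Modulators vary some layers, and a concurrency override prevents new overlapping instances.
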SoundCue /MasamuneGameplay/Assets/Audio/Stand_SC/SC_MassamuneStand_Open_Cue
root=QualityLevel_0
nodes [
 {
  id=Enveloper_1
  children [Random_0]
  properties {
   pitchCurve {editorCurveData {keys [{time=0}]}}
   volumeCurve {editorCurveData {keys [{interpMode=RCIM_Cubic, tangentMode=RCTM_Break, time=0.1, value=0.01}, {interpMode=RCIM_Cubic, tangentMode=RCTM_Break, time=0.12, value=1}]}}
  }
 }
 {id=Mixer_0, children [Enveloper_1,Modulator_3,Modulator_5,Random_2], properties {inputVolume [0.75,1.3,0.3,1.3]}}
 {id=Mixer_1, children [Modulator_0,Modulator_2,Modulator_6,Modulator_7], properties {inputVolume [1,1.5,0.5,1]}}
 {id=Modulator_0, children [WavePlayer_20], properties {pitchMin=1, volumeMax=1, volumeMin=1}}
 {id=Modulator_2, children [WavePlayer_11], properties {pitchMin=1}}
 {id=Modulator_3, children [Random_1], properties {pitchMin=1}}
 {id=Modulator_5, children [WavePlayer_23], properties {pitchMax=1.3, pitchMin=0.9, volumeMax=1, volumeMin=1}}
 {id=Modulator_6, children [WavePlayer_22], properties {pitchMax=1.3, pitchMin=0.9, volumeMax=1, volumeMin=1}}
 {id=Modulator_7, children [WavePlayer_21], properties {pitchMin=1}}
 {id=QualityLevel_0, children [Mixer_0,Mixer_0,Mixer_1,Mixer_1]}
 {id=Random_0, children [WavePlayer_0,WavePlayer_2,WavePlayer_3,WavePlayer_4], properties {weights [1,1,1,1]}}
 {id=Random_1, children [WavePlayer_7, WavePlayer_8, WavePlayer_9, WavePlayer_10], properties {weights [1,1,1,1]}}
 {id=Random_2, children [WavePlayer_5, WavePlayer_6, WavePlayer_14, WavePlayer_15], properties {weights [1,1,1,1]}}
 {id=WavePlayer_0, properties {soundWaveAssetPtr=&/MassamuneStand_Open_Bell_01}}
 {id=WavePlayer_10, properties {soundWaveAssetPtr=&/MassamuneStand_Open_Blade_Scrape_04}}
 {id=WavePlayer_11, properties {soundWaveAssetPtr=&/MassamuneStand_Open_Blade_Scrape_02}}
 {id=WavePlayer_14, properties {soundWaveAssetPtr=&/MassamuneStand_Open_Foley_03}}
 {id=WavePlayer_15, properties {soundWaveAssetPtr=&/MassamuneStand_Open_Foley_04}}
 {id=WavePlayer_2, properties {soundWaveAssetPtr=&/MassamuneStand_Open_Bell_02}}
 {id=WavePlayer_20, properties {soundWaveAssetPtr=&MassamuneStand_Open_Bell_01}}
 {id=WavePlayer_21, properties {soundWaveAssetPtr=&/MassamuneStand_Open_Foley_02}}
 {id=WavePlayer_22, properties {soundWaveAssetPtr=&/treasure_chest_open_latch_01}}
 {id=WavePlayer_23, properties {soundWaveAssetPtr=&treasure_chest_open_latch_01}}
 {id=WavePlayer_3, properties {soundWaveAssetPtr=&/MassamuneStand_Open_Bell_03}}
 {id=WavePlayer_4, properties {soundWaveAssetPtr=&/MassamuneStand_Open_Bell_04}}
 {id=WavePlayer_5, properties {soundWaveAssetPtr=&/MassamuneStand_Open_Foley_01}}
 {id=WavePlayer_6, properties {soundWaveAssetPtr=&MassamuneStand_Open_Foley_02}}
 {id=WavePlayer_7, properties {soundWaveAssetPtr=&/MassamuneStand_Open_Blade_Scrape_01}}
 {id=WavePlayer_8, properties {soundWaveAssetPtr=&MassamuneStand_Open_Blade_Scrape_02}}
 {id=WavePlayer_9, properties {soundWaveAssetPtr=&/MassamuneStand_Open_Blade_Scrape_03}}
]
properties {
 attenuationSettings=/Game/Sounds/Attenuation/Fort_Item_Interact_Att
 bOverrideConcurrency=true
 maxDistance=1850
 soundClassObject=&Fort_Loot_Reward_PSM
 volumeMultiplier=1
 concurrencyOverrides {resolutionRule=PreventNew}
}
